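# Unlit Niagara mesh-particle material for the shattered gem crystals in the DiamondHeart skydiving contrail.
# Emissive blends a matcap-like color lookup from rotated view-space normals with bump-offset triangle facets, fresnel, particle color and scrolling triangle sparkles.
# Switchable world position offset spreads, rotates, drops, wobbles and scales individual pieces, using pivots in extra UV channels and vertex-color variation.
# Niagara dynamic parameters drive piece scale, fall and offset blend.
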Material /BRCosmetics/Athena/Effects/Particles/Skydiving/NS_Contrail_DiamondHeart/M_Contrail_DiamondHeart_Crystals
graph {
 properties {
  bUsedWithNiagaraMeshParticles=true
  pixelDepthOffsetMode=PDOM_Legacy
  previewMesh=&/SM_Contrail_DiamondHeart_GemShattered
  shadingModel=MSM_Unlit
  usedShadingModels=Unlit
 }
 nodes [
  n1 = Multiply(A=TextureCoordinate()) {constB=10}
  n2 = Time()
  n3 = Constant2Vector() {g=0.1, r=0.2}
  n4 = TextureSample(UVs=Add(A=@n1, B=Multiply(A=@n2, B=@n3))) {
   samplerType=SAMPLERTYPE_Masks
   texture=&/T_Contrail_DiamondHeart_Triangles
  }
  n5 = Time()
  n6 = Constant2Vector() {g=-0.15, r=-0.2}
  n7 = TextureSample(UVs=Add(A=@n1, B=Multiply(A=@n5, B=@n6))) {
   samplerType=SAMPLERTYPE_Masks
   texture=&T_Contrail_DiamondHeart_Triangles
  }
  n8 = Constant3Vector() {constant=<rgba 0 0 1 0>}
  n9 = NamedRerouteUsage()
  n10 = Transform(None=Add(
    A=VertexNormalWS()
    B=MaterialFunctionCall(Rotation Axis=@n8, Rotation Angle=@n9) {materialFunction=&FixRotateAboutAxisNormals}.New World Space Vertex Normal
   )) {transformSourceType=TRANSFORMSOURCE_World, transformType=TRANSFORM_View}
  n11 = Add(A=ComponentMask(None=Multiply(A=@n10) {constB=0.5}) {g=true, r=true}) {constB=0.5}
  n12 = Multiply(A=Time()) {constB=0.1}
  n13 = TextureCoordinate()
  n14 = Constant() {r=-20}
  n15 = Power(Base=DotProduct(A=VertexNormalWS(), B=CameraVectorWS())) {constExponent=5}
  n16 = TextureSample(UVs=BumpOffset(Coordinate=Multiply(A=@n13) {constB=2}, Height=Multiply(A=@n14, B=@n15))) {
   samplerType=SAMPLERTYPE_Masks
   texture=&T_Contrail_DiamondHeart_Triangles
  }
  n17 = Fresnel() {exponent=2}
  n18 = LinearInterpolate(
   A=TextureSample(UVs=Add(A=@n11, B=@n12)) {texture=&/T_Contrail_DiamondHeart_Color}.RGB
   B=Constant3Vector() {constant=<rgba 0.4991 0.671 0.8854 1>}
   Alpha=Saturate(None=Subtract(A=@n16.R, B=@n17))
  )
  n19 = Add(A=Multiply(A=Multiply(A=@n4.G, B=@n7.G)) {constB=0.2}, B=Multiply(A=Add(A=@n18, B=@n17), B=ParticleColor().RGB))
  LinearInterpolate(A=@n19, B=Constant3Vector() {constant=<rgba 0.41 0.5459 0.6146 1>}, Alpha=@n15)
  MaterialFunctionCall(Emissive=@n19) {materialFunction=&MF_FX_EmissiveAdjustments}
  n20 = VertexColor()
  n21 = Multiply(A=Add(A=Time(), B=Multiply(A=@n20.G) {constB=3})) {constB=0.4}
  Multiply(A=@n21, B=LinearInterpolate(Alpha=@n20.R) {constA=0.25}) {constB=0.2}
  NamedRerouteDeclaration(None=@n21) {name=RotationAngle, nodeColor=<rgba 0.6094 0.2188 1 1>}
  n22 = Constant3Vector() {constant=<rgba 0 0 1 0>}
  n23 = DynamicParameter() {defaultValue=<rgba 1 1 0 1>, paramNames [PieceScaleAmount, LerpWPO, InitialZMovement, Param4]}
  n24 = Power(Base=VertexColor().B)
  NamedRerouteDeclaration(None=Multiply(A=Multiply(A=@n22, B=@n23.InitialZMovement), B=LinearInterpolate(Alpha=@n24) {constA=0.1})) {name=FallingPieces, nodeColor=<rgba 0.856 1 0.0405 1>}
  n25 = VectorParameter() {parameterName=WobblingAmount, defaultValue=<rgba 0 0 10 1>}
  n26 = Time()
  n27 = VertexColor()
  n28 = Multiply(A=@n27.G) {constB=3}
  n29 = Sine(None=Multiply(A=Add(A=@n26, B=@n28)) {constB=0.5})
  NamedRerouteDeclaration(None=Multiply(A=Multiply(A=@n25.RGB, B=@n29) {constB=25}, B=@n27.G)) {name=ZWobbling, nodeColor=<rgba 0.071 0.0757 1 1>}
  n30 = TextureCoordinate() {coordinateIndex=2}
  n31 = Subtract(
   A=TransformPosition(None=WorldPosition().XYZ) {transformSourceType=TRANSFORMPOSSOURCE_World, transformType=TRANSFORMPOSSOURCE_Instance}
   B=AppendVector(A=TextureCoordinate() {coordinateIndex=1}, B=ComponentMask(None=@n30) {r=true})
  )
  n32 = Multiply(
   A=LinearInterpolate(Alpha=DynamicParameter() {defaultValue=<rgba 1 1 0 1>, paramNames [PieceScaleAmount, LerpWPO, InitialZMovement, Param4]}.PieceScaleAmount) {constA=-1, constB=0}
   B=VertexColor().G
  )
  NamedRerouteDeclaration(None=Transform(None=Multiply(A=@n31, B=@n32)) {transformSourceType=TRANSFORMSOURCE_Instance}) {name=ScaleAmount, nodeColor=<rgba 1 0.0398 0.2813 1>}
  n33 = AppendVector(A=TextureCoordinate() {coordinateIndex=1}, B=Constant())
  n34 = VertexColor()
  n35 = Multiply(
   A=Normalize(VectorInput=Transform(None=@n33) {transformSourceType=TRANSFORMSOURCE_Local})
   B=Multiply(B=LinearInterpolate(Alpha=@n34.R) {constA=0.3}) {constA=70}
  ) {constB=65}
  n36 = NamedRerouteUsage()
  n37 = ParticlePositionWS()
  n38 = Add(
   A=Add(A=Add(A=@n35, B=@n36), B=NamedRerouteUsage())
   B=RotateAboutAxis(
    NormalizedRotationAxis=Constant3Vector() {constant=<rgba 0 0 1 0>}
    RotationAngle=NamedRerouteUsage()
    PivotPoint=TransformPosition(None=@n37) {transformSourceType=TRANSFORMPOSSOURCE_World, transformType=TRANSFORMPOSSOURCE_Instance}
    Position=@n35
   )
  )
  n39 = NamedRerouteUsage()
  StaticSwitchParameter(
   True=LinearInterpolate(
    A=Add(A=@n38, B=@n39)
    Alpha=DynamicParameter() {defaultValue=<rgba 1 1 0 1>, paramNames [PieceScaleAmount, LerpWPO, InitialZMovement, Param4]}.LerpWPO
   ) {constB=0}
   False=Constant()
  ) {defaultValue=true, parameterName=UseWPO}
 ]
}
