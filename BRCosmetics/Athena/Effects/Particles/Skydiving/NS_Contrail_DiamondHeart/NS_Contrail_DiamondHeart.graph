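# Niagara skydiving contrail cosmetic, DiamondHeart, that attaches gems, glows, sparkles and ribbons to the player's skeletal mesh.
# Two sampled bones drive paired white and purple ribbons, stars and lit hand glows; a third anchors persistent gem meshes, shards, bokeh, a middle ribbon and GPU orbiting sprites.
# ParaGlideLeanAlpha crossfades center elements, shrinking gem parts and orbiters while brightening spinning glows.
# Dive state accelerates ribbons, and time of day tints gems.
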NiagaraSystem /BRCosmetics/Athena/Effects/Particles/Skydiving/NS_Contrail_DiamondHeart/NS_Contrail_DiamondHeart
dependencies {
 usedDataInterfaces [
  &NiagaraDataInterfaceArrayFloat
  &NiagaraDataInterfaceCurve
  &NiagaraDataInterfaceColorCurve
  &NiagaraDataInterfaceMeshRendererInfo
 ]
 usedDynamicInputs [&SimulationPosition]
 usedModules [&SystemState]
}
emitters {
 Bokeh {
  emitterUpdateScript [
   &EmitterState()
   &SpawnRate(SpawnRate=2, Use Spawn Probability=false)
  ]
  particleSpawnScript [
   &InitializeParticle(
    AdjustAlpha=false
    AdjustHue=true
    AdjustSaturation=false
    AdjustValue=false
    Color=<rgba 0.751 0.6837 1 0.1>
    Color Mode=::NewEnumerator3
    Lifetime Max=1.2
    Lifetime Min=0.6
    Lifetime Mode=::NewEnumerator1
    Position Offset=<0 -45 105>
    Sprite Size Mode=::NewEnumerator1
    Uniform Sprite Size Max=80
    Uniform Sprite Size Min=40
    UsePositionOffset=false
    Write Material Random=true
   ) {module=InitializeParticle001}
   &SkeletalMeshLocation(
    Bone / Socket Index=2
    Bone Sampling Mode=::NewEnumerator4
    Interpolate Mesh Sampling=false
    MaskSpawnGroup=false
    Sampled Position Offset=<0 -35 0>
    Skeletal Mesh=$User.SkeletalMesh
   )
   &ShapeLocation(
    Box Size=<80 80 50>
    MaskSpawnGroup=false
    Override Seed=false
    Shape Primitive=::NewEnumerator2
    Surface Only=true
    Surface Only Band Thickness=1
    UseOwnerScale=true
   )
  ]
  particleUpdateScript [
   &ParticleState()
   &SpriteRotationRate(Rotation Rate=&RandomRangeFloat(Minimum=-360, Maximum=360))
   &ScaleSpriteSize(Uniform Curve Sprite Scale=DataInterface() {propertyValues {Curve {keys [{tangentMode=RCTM_User, time=0, value=0.9}, <at 0.2 1>, <at 0.6 1>, {tangentMode=RCTM_User, time=1, value=0.8}]}}})
   &CameraOffset(Camera Offset Amount=20)
   &Drag(Drag=1.25, Use Linear Drag=true, Use Rotational Drag=true)
   &SolveForcesAndVelocity(Clamp Velocity=false, Limit Acceleration=false)
   &ScaleColor(
    Scale Alpha=&FloatFromCurve(FloatCurve=DataInterface() {propertyValues {Curve {keys [
         {tangentMode=RCTM_User, time=0, value=0.05}
         {arriveTangent=-1.6723, interpMode=RCIM_Cubic, leaveTangent=-1.6723, tangentMode=RCTM_User, time=0.3, value=1.025}
         {arriveTangent=-1.1836, interpMode=RCIM_Cubic, leaveTangent=-1.1836, tangentMode=RCTM_User, time=1, value=0}
        ]}}})
    ScaleA=true
    ScaleRGB=true
   )
  ]
  rendererClasses [&NiagaraSpriteRendererProperties]
  renderers [{rendererClass=&NiagaraSpriteRendererProperties}]
 }
 Empty {
  simTarget=GPUComputeSim
  emitterUpdateScript [
   &EmitterState()
   &SpawnRate(SpawnRate=150, Use Spawn Probability=false)
  ]
  particleSpawnScript [&InitializeParticle(
    AdjustAlpha=false
    AdjustHue=true
    AdjustSaturation=false
    AdjustValue=false
    Color=<rgba 0.5284 0.6322 1 1>
    Color Mode=::NewEnumerator3
    Lifetime Max=0.8
    Lifetime Min=0.3
    Lifetime Mode=::NewEnumerator1
    Sprite Size Mode=::NewEnumerator1
    Uniform Sprite Size Max=&RandomRangeFloat(Minimum=2, Maximum=4)
    Uniform Sprite Size Min=1
    UsePositionOffset=false
    Write Material Random=true
   )]
  particleUpdateScript [
   &ParticleState()
   &SkeletalMeshLocation(
    Bone / Socket Index=2
    Bone Sampling Mode=::NewEnumerator4
    MaskSpawnGroup=false
    Sampled Position Offset=<10 -35 0>
    Sampled Position Offset Coordinate Space:ENiagaraSkelMeshTransforms=::NewEnumerator4
    Skeletal Mesh=$User.SkeletalMesh
   )
   &ScaleSpriteSize(Scale Sprite Size Mode=::NewEnumerator0, Uniform Scale Factor=&OneMinusFloat(Float=$User.ParaGlideLeanAlpha))
   &SolveForcesAndVelocity(Clamp Velocity=false, Limit Acceleration=false)
   &RotateAroundPoint(
    Radius:NiagaraFloat=&RandomRangeFloat(Minimum=75, Maximum=5)
    Rotation Center:NiagaraPosition=&AddVectorToPosition(
     Position=$Particles.SkeletalMeshLocation.SampledPosition
     Vector=&RandomRangeVector(Minimum=<0 0 30>, Maximum=<0 0 80>)
    )
    Rotation Phase:NiagaraFloat=&RandomRangeFloat(Minimum=-1, Maximum=1)
    Rotation Rate:NiagaraFloat=&SelectFloatFromArray(
     Array Sampling Mode=::NewEnumerator0
     Direct Array Index=?
     Float Selection Array=DataInterface() {propertyValues {FloatData [-0.5,0.5]}}
     Random Seed=0
    )
   )
   &CameraOffset(Camera Offset Amount=-15)
  ]
  rendererClasses [&NiagaraSpriteRendererProperties]
  renderers [{rendererClass=&NiagaraSpriteRendererProperties}]
 }
 GemPieces {
  emitterUpdateScript [
   &EmitterState()
   &SpawnRate(SpawnRate=&Lerp_Float(A=8, B=0, Alpha=$User.ParaGlideLeanAlpha), Use Spawn Probability=false)
  ]
  particleSpawnScript [
   &InitializeParticle(
    AdjustAlpha=false
    AdjustHue=true
    AdjustSaturation=false
    AdjustValue=false
    Color=<rgba 0.6983 0.5853 1.1 1>
    Color Mode=::NewEnumerator3
    Hue Shift Range:Vector2f=<-0.05 0.05>
    Lifetime Max=0.6
    Lifetime Min=0.35
    Lifetime Mode=::NewEnumerator1
    Mesh Renderer Array Visibility Mode=::NewEnumerator3
    Mesh Renderer Info=DataInterface() {propertyValues {MeshRenderer=&self:GemPieces_13.NiagaraMeshRendererProperties_0}}
    Mesh Scale Max=<1 1 0.2>
    Mesh Scale Min=<0.4 0.4 0.1>
    Mesh Scale Mode=::NewEnumerator2
    Position Offset=<0 -45 105>
    UsePositionOffset=false
    Write Material Random=true
   ) {module=InitializeParticle001}
   &SkeletalMeshLocation(
    Bone / Socket Index=2
    Bone Sampling Mode=::NewEnumerator4
    Interpolate Mesh Sampling=false
    MaskSpawnGroup=false
    Sampled Position Offset=<0 -35 0>
    Skeletal Mesh=$User.SkeletalMesh
   )
   &ShapeLocation(
    Box Midpoint=<0.5 0.5 0>
    Box Size=<60 50 2>
    MaskSpawnGroup=false
    Override Seed=false
    Shape Primitive=::NewEnumerator2
    Surface Only=true
    Surface Only Band Thickness=1
    UseOwnerScale=true
   )
   &AddVelocity(Velocity=&RandomRangeVector(Minimum=<0 0 160>, Maximum=<1 1 300>))
   &InitialMeshOrientation(Use Rotation Vector=false)
   &DynamicMaterialParameters(Module.Index0Param0Write=true, Module.Index0Param1Write=true, Module.Index0Param2Write=true, Module.Index0Param3Write=true)
  ]
  particleUpdateScript [
   &ParticleState()
   &ScaleMeshSize(Scale Factor=&VectorFromFloat(Value=&FloatFromCurve(FloatCurve=DataInterface() {propertyValues {Curve {keys [<at 0 0.2>,<at 0.3 1>,<at 0.8 0.85>,<at 1 0>]}}})))
   &UpdateMeshOrientation(Rotation Rate=0.8, Rotation Vector=<1 1 1>)
   &CameraOffset(Camera Offset Amount=10)
   &Drag(Drag=1.25, Use Linear Drag=true, Use Rotational Drag=true)
   &SolveForcesAndVelocity(Clamp Velocity=false, Limit Acceleration=false)
   &Color(Color=&MultiplyLinearColor(
     Linear Color=$Particles.Initial.Color
     Scale Factor=&MakeVector4_FromLinearColor(LinearColor=&FN_Select_Color_from_Time_Of_Day(Morning=<rgba 1 0.5975 0.6675 1>, Day=<rgba 1 1 1 1>, Evening=<rgba 1 1 1 1>, Night=<rgba 0.645 0.5825 0.8646 1>))
    ))
  ]
  rendererClasses [&NiagaraMeshRendererProperties]
  renderers [{rendererClass=&NiagaraMeshRendererProperties, rendererIndex=0}]
 }
 Gems001 {
  emitterUpdateScript [
   &EmitterState(Life Cycle Mode=::NewEnumerator1, Loop Behavior=::NewEnumerator1, Loop Duration=1, UseLoopDelay=false)
   &SpawnBurst_Instantaneous(Spawn Count=1, Spawn Probability=HlslExpression() {hlslExpression="Emitter.LoopCount == 0"}, Use Spawn Probability=true)
  ]
  particleSpawnScript [
   &FN_SetSpawnGroup(Number of Spawn Groups=1)
   &InitializeParticle(
    Color=<rgba 0.3759 0.4681 0.5938 1>
    Lifetime=1.25
    Mass Mode=::NewEnumerator1
    Mesh Scale Mode=::NewEnumerator3
    UsePositionOffset=false
    Write Material Random=true
   ) {module=InitializeParticle001}
   /SetVariables(Particles.Render2Scale:Vector3f=<0.2 0.3 0.3>) {isSetParameters=true}
  ]
  particleUpdateScript [
   &ParticleState(Kill Particles When Lifetime Has Elapsed=false)
   &ScaleMeshSize(Scale Factor=&VectorFromFloat(Value=&FloatFromCurve(FloatCurve=DataInterface() {propertyValues {Curve {keys [{interpMode=RCIM_Cubic, leaveTangent=6.3424, tangentMode=RCTM_Break, time=0, value=0}, <at 0.2 0.8>]}}}))) {enabled=false}
   &SkeletalMeshLocation(
    Bone / Socket Index=2
    Bone Sampling Mode=::NewEnumerator4
    MaskSpawnGroup=false
    Orientation Sampling=::NewEnumerator1
    Sampled Position Offset=<-10 -30 0>
    Skeletal Mesh=$User.SkeletalMesh
   )
   &DynamicMaterialParameters(
    Index 0 Param 1=0.4
    Index 0 Param 2=$User.ParaGlideLeanAlpha
    Index 0 Param 3=&FloatFromCurve(
     FloatCurve=DataInterface() {propertyValues {Curve {keys [
         {interpMode=RCIM_Cubic, leaveTangent=-0.3999, tangentMode=RCTM_Break, time=0, value=1}
         {arriveTangent=0.0515, interpMode=RCIM_Cubic, leaveTangent=36.7092, tangentMode=RCTM_Break, time=0.6, value=0}
        ]}}}
     Scale Curve=1000
    )
    Module.Index0Param0Write=true
    Module.Index0Param1Write=true
    Module.Index0Param2Write=true
    Module.Index0Param3Write=true
   )
   /SetVariables(
    Particles.Render2Pos:NiagaraPosition=&AddVectorToPosition(Position=$Particles.Position, Vector=<0 3 10>)
    Particles.Render2Scale:Vector3f=&Multiply_VectorByFloat(Vector=$Particles.Initial.Render2Scale, Float=&OneMinusFloat(Float=$User.ParaGlideLeanAlpha))
   ) {isSetParameters=true}
   &Color(Color=&MultiplyLinearColor(
     Linear Color=$Particles.Initial.Color
     Scale Factor=&MakeVector4_FromLinearColor(LinearColor=&FN_Select_Color_from_Time_Of_Day(Morning=<rgba 1 0.5975 0.6675 1>, Day=<rgba 1 1 1 1>, Evening=<rgba 1 1 1 1>, Night=<rgba 0.645 0.5825 0.8646 1>))
    ))
  ]
  rendererClasses [&NiagaraMeshRendererProperties]
  renderers [
   {rendererClass=&NiagaraMeshRendererProperties, rendererIndex=0}
   {rendererClass=&NiagaraMeshRendererProperties, rendererIndex=1}
  ]
 }
 GemsStars {
  emitterUpdateScript [
   &EmitterState()
   &SpawnRate(SpawnRate=4, Use Spawn Probability=false)
  ]
  particleSpawnScript [
   &InitializeParticle(
    AdjustAlpha=false
    AdjustHue=true
    AdjustSaturation=false
    AdjustValue=false
    Color=<rgba 1.0812 1.0916 3 1>
    Color Mode=::NewEnumerator3
    Hue Shift Range:Vector2f=<-0.3 0.3>
    Lifetime Max=0.5
    Lifetime Min=0.25
    Lifetime Mode=::NewEnumerator1
    Position Offset=<0 -45 105>
    Sprite Size Mode=::NewEnumerator1
    Uniform Sprite Size Max=35
    Uniform Sprite Size Min=20
    UsePositionOffset=false
    Write Material Random=true
   ) {module=InitializeParticle001}
   &SkeletalMeshLocation(
    Bone / Socket Index=2
    Bone Sampling Mode=::NewEnumerator4
    Interpolate Mesh Sampling=false
    MaskSpawnGroup=false
    Sampled Position Offset=<0 -35 0>
    Skeletal Mesh=$User.SkeletalMesh
   )
   &ShapeLocation(
    Box Size=<100 80 35>
    MaskSpawnGroup=false
    Override Seed=false
    Shape Primitive=::NewEnumerator2
    Surface Only=true
    Surface Only Band Thickness=1
    UseOwnerScale=true
   )
  ]
  particleUpdateScript [
   &ParticleState()
   &SpriteRotationRate(Rotation Rate=&RandomRangeFloat(Minimum=-360, Maximum=360)) {enabled=false}
   &ScaleSpriteSize(Uniform Curve Sprite Scale=DataInterface() {propertyValues {Curve {keys [
        {interpMode=RCIM_Cubic, leaveTangent=3.2871, tangentMode=RCTM_Break, time=0, value=0}
        <at 0.5 1>
        <at 0.7 1>
        {tangentMode=RCTM_User, time=1, value=0.15}
       ]}}})
   &CameraOffset(Camera Offset Amount=10)
   &Drag(Drag=1.25, Use Linear Drag=true, Use Rotational Drag=true)
   &SolveForcesAndVelocity(Clamp Velocity=false, Limit Acceleration=false)
  ]
  rendererClasses [&NiagaraSpriteRendererProperties]
  renderers [{rendererClass=&NiagaraSpriteRendererProperties}]
 }
 HandGlow {
  emitterUpdateScript [
   &EmitterState(Life Cycle Mode=::NewEnumerator1, Loop Behavior=::NewEnumerator1, Loop Duration Mode=::NewEnumerator1, UseLoopDelay=false)
   &SpawnBurst_Instantaneous(Spawn Count=4, Spawn Probability=HlslExpression() {hlslExpression="Emitter.LoopCount == 0"}, Use Spawn Probability=true)
  ]
  particleSpawnScript [
   &FN_SetSpawnGroup(Number of Spawn Groups=2)
   &InitializeParticle(
    Color=<rgba 2 2 2 0.25>
    Lifetime=2
    Mass Mode=::NewEnumerator1
    Sprite Size Mode=::NewEnumerator3
    Uniform Sprite Size=65
    UsePositionOffset=false
    Write Material Random=true
   ) {module=InitializeParticle001}
  ]
  particleUpdateScript [
   &ParticleState(Kill Particles When Lifetime Has Elapsed=false)
   &SkeletalMeshLocation(
    Bone / Socket Index=$Particles.SpawnGroup
    Bone Sampling Mode=::NewEnumerator4
    Inherit Velocity=false
    Interpolate Mesh Sampling=false
    MaskSpawnGroup=false
    Sampled Position Offset=<0 0 7>
    Skeletal Mesh=$User.SkeletalMesh
   )
   &CameraOffset(Camera Offset Amount=15)
   &ScaleColor(
    Curve Index=$System.LoopedAge
    Linear Color Curve=DataInterface() {propertyValues {
      AlphaCurve {keys [<at 0 1>,<at 0.4671 0.6>,<at 1 1>]}
      BlueCurve {keys [<at 0.1217 0.8714>,<at 0.3306 1>,<at 0.4729 0.7292>,<at 0.6336 0.8714>,<at 0.8205 0.8714>]}
      GreenCurve {keys [<at 0.1217 0.3515>,<at 0.3306 0.3241>,<at 0.4729 0.2521>,<at 0.6336 0.1894>,<at 0.8205 0.3515>]}
      RedCurve {keys [<at 0.1217 0.227>,<at 0.3306 0.3406>,<at 0.4729 0.1822>,<at 0.6336 0.2246>,<at 0.8205 0.227>]}
     }}
    Scale Mode=::NewEnumerator2
   )
   &FN_Light_v2(Light Exponent=2, Light Radius=100, LightEnabled=true, Scale Color=2)
  ]
  rendererClasses [&NiagaraSpriteRendererProperties, &NiagaraLightRendererProperties]
  renderers [
   {rendererClass=&NiagaraSpriteRendererProperties}
   {rendererClass=&NiagaraLightRendererProperties, rendererIndex=1}
  ]
 }
 HandGlow001 {
  emitterUpdateScript [
   &EmitterState(Life Cycle Mode=::NewEnumerator1, Loop Behavior=::NewEnumerator1, Loop Duration Mode=::NewEnumerator1, UseLoopDelay=false)
   &SpawnBurst_Instantaneous(Spawn Count=3, Spawn Probability=HlslExpression() {hlslExpression="Emitter.LoopCount == 0"}, Use Spawn Probability=true)
  ]
  particleSpawnScript [
   &FN_SetSpawnGroup(Number of Spawn Groups=3)
   &InitializeParticle(
    AdjustAlpha=false
    AdjustHue=true
    AdjustSaturation=false
    AdjustValue=false
    Color=<rgba 0.598 0.6062 1 1>
    Color Mode=::NewEnumerator3
    Lifetime=2
    Mass Mode=::NewEnumerator1
    Position Offset=<0 -30 110>
    Sprite Size Mode=::NewEnumerator3
    Uniform Sprite Size=130
    UsePositionOffset=false
    Write Material Random=true
   ) {module=InitializeParticle001}
  ]
  particleUpdateScript [
   &ParticleState(Kill Particles When Lifetime Has Elapsed=false)
   &SpriteRotationRate(Rotation Rate=&SelectFloatFromArray(Direct Array Index=$Particles.SpawnGroup, Float Selection Array=DataInterface() {propertyValues {FloatData [-60,60,-120]}}))
   &ScaleColor(
    Curve Index=$System.LoopedAge
    Linear Color Curve=DataInterface() {propertyValues {
      AlphaCurve {keys [<at 0 1>,<at 0.4671 0.6>,<at 1 1>]}
      BlueCurve {keys [<at 0.1217 0.7605>,<at 0.3306 1>,<at 0.4729 0.7292>,<at 0.6336 0.8714>,<at 0.8205 0.7605>]}
      GreenCurve {keys [<at 0.1217 0.6308>,<at 0.3306 0.9701>,<at 0.4729 0.4584>,<at 0.6336 0.5208>,<at 0.8205 0.6308>]}
      RedCurve {keys [<at 0.1217 0.4287>,<at 0.3306 0.7067>,<at 0.4729 0.3955>,<at 0.6336 0.5719>,<at 0.8205 0.4287>]}
     }}
    Scale Mode=::NewEnumerator2
   )
   &DynamicMaterialParameters(
    Index 0 Param 1=1
    Index 0 Param 2=&RandomRangeFloat(Minimum=0, Maximum=0.2)
    Module.Index0Param0Write=true
    Module.Index0Param1Write=true
    Module.Index0Param2Write=true
    Module.Index0Param3Write=true
   )
   &Color(Color=&MultiplyLinearColor(Linear Color=$Particles.Color, Scale Factor=&MakeVector4_FromFloat(Float=$User.ParaGlideLeanAlpha)))
   &SkeletalMeshLocation(
    Bone / Socket Index=2
    Bone Sampling Mode=::NewEnumerator4
    MaskSpawnGroup=false
    Orientation Sampling=::NewEnumerator1
    Sampled Position Offset=&FN_FrontEndVector3Modifier(Gameplay=<0 -30 0>, Front End=<-15 -45 0>)
    Skeletal Mesh=$User.SkeletalMesh
   )
   &CameraOffset()
  ]
  rendererClasses [&NiagaraSpriteRendererProperties]
  renderers [{rendererClass=&NiagaraSpriteRendererProperties}]
 }
 MiddleRibbon {
  emitterUpdateScript [
   &EmitterState()
   /Game/Effects/Niagara/Modules/Skydiving/FN_DivingAttributes/FN_EmitterDivingAttributes_v1(Bone Index=1, SkeletalMesh:NiagaraDataInterfaceSkeletalMesh=$User.SkeletalMesh)
   &FN_SpawnRibbon_v2(Spawn Mode=::NewEnumerator0, SpawnRate=128)
  ]
  particleSpawnScript [
   &FN_SetSpawnGroup(Number of Spawn Groups=1)
   &FN_InitializeRibbon_v3(Color:LinearColor=<rgba 1 1 1 0.8>, Enable Skeletal Position=true, Number of Bones Per Boneflow group=0, Ribbon Width=18)
   &SkeletalMeshLocation(
    Apply Sampled Orientation to Mesh Particles=true
    Bone / Socket Index=2
    Bone Sampling Mode=::NewEnumerator4
    Inherit Velocity=false
    Interpolate Mesh Sampling=false
    MaskSpawnGroup=false
    Sampled Position Offset=<-5 -30 0>
    Sampled Position Offset Coordinate Space:ENiagaraMeshTransforms=::NewEnumerator3
    Skeletal Mesh=$User.SkeletalMesh
   )
   &AddVelocity(
    Scale Added Velocity=&Lerp_Vector(A=<1 1 1>, B=<4 4 1>, Alpha=$System.Dive)
    Velocity=&Multiply_Vector(
     A=&FN_FrontEndVector3Modifier(Gameplay=<0 0 350>, Front End=<0 0 350>)
     B=&Lerp_Vector(A=<0 0 1.2>, B=<0 0 0>, Alpha=$User.ParaGlideLeanAlpha)
    )
   )
   &DynamicMaterialParameters(
    Index 0 Param 1=&MaskFloatBySpawnGroup(Masked Float=1.25, Passthrough Float=0.8)
    Index 0 Param 2=&MaskFloatBySpawnGroup(Masked Float=0.5, Passthrough Float=3)
    Index 0 Param 3=1
    Index 1 Param 1=&Lerp_Float(A=3, B=6, Alpha=$System.Dive)
    Module.Param0WriteEnabled=true
    Module.Param0WriteEnabled1=true
    Module.Param1WriteEnabled=true
    Module.Param1WriteEnabled1=false
    Module.Param2WriteEnabled=true
    Module.Param2WriteEnabled1=false
    Module.Param3WriteEnabled=false
    Module.Param3WriteEnabled1=false
    Write Parameter Index 1=true
   )
  ]
  particleUpdateScript [
   &InheritSourceMovement(Applied Position Falloff Distance=100, ApplyForce=false, ApplyPosition=true, ApplyPositionFalloff=false)
   &SolveForcesAndVelocity(Clamp Velocity=false, Limit Acceleration=false)
   &FN_UpdateRibbon(Detach Locked Particle=true, Enable Skeletal Position=true)
   &ScaleRibbonWidth(Ribbon Width Scale=&FloatFromCurve(FloatCurve=DataInterface() {propertyValues {Curve {keys [
         {interpMode=RCIM_Cubic, leaveTangent=11.3372, tangentMode=RCTM_Break, time=0, value=0.6}
         {arriveTangent=1.1826, interpMode=RCIM_Cubic, leaveTangent=-5.8218, tangentMode=RCTM_Break, time=0.1, value=1.4}
         <at 0.8 0>
        ]}}}))
   &ScaleColor(
    Linear Color Curve=DataInterface() {propertyValues {
      AlphaCurve {keys [<at 0 1>,<at 0.2084 0.4>,<at 0.5029 0>]}
      BlueCurve {keys [<at 0 8>,<at 0.4086 1>,<at 1 1>]}
      GreenCurve {keys [<at 0 4.6823>,<at 0.4086 0.7333>,<at 1 1>]}
      RedCurve {keys [<at 0 5.7002>,<at 0.4086 0.6537>,<at 1 1>]}
     }}
    Scale Mode=::NewEnumerator2
   )
  ]
  rendererClasses [&NiagaraRibbonRendererProperties]
  renderers [{rendererClass=&NiagaraRibbonRendererProperties}]
 }
 RibbonPurple {
  emitterUpdateScript [
   &EmitterState()
   &FN_EmitterDivingAttributes_v1(Bone Index=1, SkeletalMesh:NiagaraDataInterfaceSkeletalMesh=$User.SkeletalMesh)
   &FN_SpawnRibbon_v2(Number of Ribbons=2, Spawn Mode=::NewEnumerator0, SpawnRate=64)
  ]
  particleSpawnScript [
   &FN_SetSpawnGroup(Number of Spawn Groups=2)
   &FN_InitializeRibbon_v3(
    Color:LinearColor=<rgba 0.7898 0.9039 1.5 0.4>
    Custom Offset=true
    Enable Skeletal Position=true
    Local Vector Offset:Vector3f=<0 1 0>
    Ribbon Width=50
   )
   &AddVelocity(
    Scale Added Velocity=&Lerp_Vector(A=<1 1 1>, B=<4 4 1>, Alpha=$System.Dive)
    Velocity=&FN_FrontEndVector3Modifier(Gameplay=<0 0 650>, Front End=<0 0 650>)
   )
   &AddVelocity(Velocity=&MakeVector(
     X=&SelectFloatFromArray(Direct Array Index=$Particles.SpawnGroup, Float Selection Array=DataInterface() {propertyValues {FloatData [100,-100]}})
     Y=0
     Z=0
    )) {module=AddVelocity001}
   &SkeletalMeshLocation(
    Apply Sampled Orientation to Mesh Particles=true
    Bone / Socket Index=$Particles.SpawnGroup
    Bone Sampling Mode=::NewEnumerator4
    Inherit Velocity=false
    Interpolate Mesh Sampling=false
    MaskSpawnGroup=false
    Sampled Position Offset Coordinate Space:ENiagaraMeshTransforms=::NewEnumerator3
    Skeletal Mesh=$User.SkeletalMesh
   )
   &DynamicMaterialParameters(
    Index 0 Param 1=&MaskFloatBySpawnGroup(Masked Float=1, Passthrough Float=0.8)
    Index 0 Param 2=&MaskFloatBySpawnGroup(Masked Float=0.4, Passthrough Float=3)
    Index 0 Param 3=1
    Index 1 Param 1=&Lerp_Float(A=5, B=10, Alpha=$System.Dive)
    Module.Param0WriteEnabled=true
    Module.Param0WriteEnabled1=true
    Module.Param1WriteEnabled=true
    Module.Param1WriteEnabled1=false
    Module.Param2WriteEnabled=true
    Module.Param2WriteEnabled1=false
    Module.Param3WriteEnabled=false
    Module.Param3WriteEnabled1=false
    Write Parameter Index 1=true
   )
  ]
  particleUpdateScript [
   &InheritSourceMovement(Applied Position Falloff Distance=100, Applied Position Scale=<0.9 0.9 1>, ApplyForce=false, ApplyPosition=true, ApplyPositionFalloff=false)
   &SolveForcesAndVelocity(Clamp Velocity=false, Limit Acceleration=false)
   &FN_UpdateRibbon(Detach Locked Particle=true, Enable Skeletal Position=true)
   &ScaleRibbonWidth(Ribbon Width Scale=&FloatFromCurve(FloatCurve=DataInterface() {propertyValues {Curve {keys [{interpMode=RCIM_Cubic, leaveTangent=1.3723, tangentMode=RCTM_Break, time=0, value=0.7}, <at 0.9 1.25>]}}}))
  ]
  rendererClasses [&NiagaraRibbonRendererProperties]
  renderers [{rendererClass=&NiagaraRibbonRendererProperties}]
 }
 RibbonWhite {
  emitterUpdateScript [
   &EmitterState()
   &FN_EmitterDivingAttributes_v1(Bone Index=1, SkeletalMesh:NiagaraDataInterfaceSkeletalMesh=$User.SkeletalMesh)
   &FN_SpawnRibbon_v2(Number of Ribbons=2, Spawn Mode=::NewEnumerator0, SpawnRate=64)
  ]
  particleSpawnScript [
   &FN_SetSpawnGroup(Number of Spawn Groups=2)
   &FN_InitializeRibbon_v3(
    Color:LinearColor=<rgba 0.7794 0.8779 0.8854 0.8>
    Custom Offset=true
    Enable Skeletal Position=true
    Local Vector Offset:Vector3f=<0 1 0>
    Ribbon Width=25
   )
   &AddVelocity(
    Scale Added Velocity=&Lerp_Vector(A=<1 1 1>, B=<4 4 1>, Alpha=$System.Dive)
    Velocity=&FN_FrontEndVector3Modifier(Gameplay=<0 0 550>, Front End=<0 0 550>)
   )
   &AddVelocity(Velocity=&MakeVector(
     X=&SelectFloatFromArray(Direct Array Index=$Particles.SpawnGroup, Float Selection Array=DataInterface() {propertyValues {FloatData [100,-100]}})
     Y=0
     Z=0
    )) {module=AddVelocity001}
   &SkeletalMeshLocation(
    Apply Sampled Orientation to Mesh Particles=true
    Bone / Socket Index=$Particles.SpawnGroup
    Bone Sampling Mode=::NewEnumerator4
    Inherit Velocity=false
    Interpolate Mesh Sampling=false
    MaskSpawnGroup=false
    Sampled Position Offset Coordinate Space:ENiagaraMeshTransforms=::NewEnumerator3
    Skeletal Mesh=$User.SkeletalMesh
   )
   &DynamicMaterialParameters(
    Index 0 Param 1=&MaskFloatBySpawnGroup(Masked Float=1, Passthrough Float=0.8)
    Index 0 Param 2=&MaskFloatBySpawnGroup(Masked Float=0.4, Passthrough Float=3)
    Index 0 Param 3=1
    Index 1 Param 1=&Lerp_Float(A=5, B=10, Alpha=$System.Dive)
    Module.Param0WriteEnabled=true
    Module.Param0WriteEnabled1=true
    Module.Param1WriteEnabled=true
    Module.Param1WriteEnabled1=false
    Module.Param2WriteEnabled=true
    Module.Param2WriteEnabled1=false
    Module.Param3WriteEnabled=false
    Module.Param3WriteEnabled1=false
    Write Parameter Index 1=true
   )
  ]
  particleUpdateScript [
   &InheritSourceMovement(Applied Position Falloff Distance=100, Applied Position Scale=<0.9 0.9 1>, ApplyForce=false, ApplyPosition=true, ApplyPositionFalloff=false)
   &SolveForcesAndVelocity(Clamp Velocity=false, Limit Acceleration=false)
   &FN_UpdateRibbon(Detach Locked Particle=true, Enable Skeletal Position=true)
   &ScaleRibbonWidth(Ribbon Width Scale=&FloatFromCurve(FloatCurve=DataInterface() {propertyValues {Curve {keys [{interpMode=RCIM_Cubic, leaveTangent=-0.3625, tangentMode=RCTM_Break, time=0, value=1}, <at 0.9 0.25>]}}}))
   &ScaleColor(
    Linear Color Curve=DataInterface() {propertyValues {
      AlphaCurve {keys [<at 0 1>,<at 0.4763 0.4>,<at 0.7785 0>]}
      BlueCurve {keys [<at 0 8>,<at 0.4086 1>,<at 1 1>]}
      GreenCurve {keys [<at 0 3.7102>,<at 0.4086 0.8468>,<at 1 1>]}
      RedCurve {keys [<at 0 4.1337>,<at 0.4086 0.5938>,<at 1 1>]}
     }}
    Scale Mode=::NewEnumerator2
   )
  ]
  rendererClasses [&NiagaraRibbonRendererProperties]
  renderers [{rendererClass=&NiagaraRibbonRendererProperties}]
 }
 Stars {
  emitterUpdateScript [
   &EmitterState()
   /Game/Effects/Niagara/Modules/Spawn/SpawnRateFromMultipleSpawnGroups_2(Global Spawnrate=25)
  ]
  particleSpawnScript [
   &FN_SetSpawnGroup(Number of Spawn Groups=2)
   &InitializeParticle(
    AdjustAlpha=false
    AdjustHue=true
    AdjustSaturation=false
    AdjustValue=false
    Color=<rgba 0.4293 0.6784 1.25 1>
    Color Mode=::NewEnumerator3
    Hue Shift Range:Vector2f=<-0.6 0.6>
    Lifetime Max=0.5
    Lifetime Min=0.15
    Lifetime Mode=::NewEnumerator1
    Sprite Size Mode=::NewEnumerator1
    Uniform Sprite Size Max=20
    Uniform Sprite Size Min=12
    UsePositionOffset=false
    Write Material Random=true
   ) {module=InitializeParticle001}
   &SkeletalMeshLocation(
    Bone / Socket Index=$Particles.SpawnGroup
    Bone Sampling Mode=::NewEnumerator4
    Inherit Velocity=false
    Interpolate Mesh Sampling=false
    MaskSpawnGroup=false
    Skeletal Mesh=$User.SkeletalMesh
   )
   &ShapeLocation(MaskSpawnGroup=false, Override Seed=false, Sphere Radius=20, UseOwnerScale=true)
   &AddVelocity(
    Velocity=&RandomRangeVector(Minimum=<0 0 100>, Maximum=<0 0 350>)
    Velocity Speed Scale=&Lerp_Float(A=1, B=1.5, Alpha=$User.ParaGlideLeanAlpha)
   )
   &AddVelocity(Velocity=&MakeVector(
     X=&SelectFloatFromArray(Direct Array Index=$Particles.SpawnGroup, Float Selection Array=DataInterface() {propertyValues {FloatData [50,-50]}})
     Y=0
     Z=0
    )) {module=AddVelocity001}
  ]
  particleUpdateScript [
   &ParticleState()
   &ScaleSpriteSize(Uniform Curve Sprite Scale=DataInterface() {propertyValues {Curve {keys [{tangentMode=RCTM_User, time=0, value=0}, <at 0.4 1>, <at 0.6 1>, {tangentMode=RCTM_User, time=1, value=0.15}]}}})
   &CameraOffset(Camera Offset Amount=10)
   &Drag(Drag=1.25, Use Linear Drag=true, Use Rotational Drag=true)
   &SolveForcesAndVelocity(Clamp Velocity=false, Limit Acceleration=false)
  ]
  rendererClasses [&NiagaraSpriteRendererProperties]
  renderers [{rendererClass=&NiagaraSpriteRendererProperties}]
 }
}
summary {
 systemName=NS_Contrail_DiamondHeart
 emitters [
  {emitterName=RibbonWhite, rendererClasses [&NiagaraRibbonRendererProperties]}
  {emitterName=RibbonPurple, rendererClasses [&NiagaraRibbonRendererProperties]}
  {emitterName=Stars, rendererClasses [&NiagaraSpriteRendererProperties]}
  {emitterName=HandGlow, rendererClasses [&NiagaraSpriteRendererProperties, &NiagaraLightRendererProperties]}
  {emitterName=MiddleRibbon, rendererClasses [&NiagaraRibbonRendererProperties]}
  {emitterName=GemsStars, rendererClasses [&NiagaraSpriteRendererProperties]}
  {emitterName=Bokeh, rendererClasses [&NiagaraSpriteRendererProperties]}
  {emitterName=GemPieces, rendererClasses [&NiagaraMeshRendererProperties]}
  {emitterName=Gems001, rendererClasses [&NiagaraMeshRendererProperties]}
  {emitterName=HandGlow001, rendererClasses [&NiagaraSpriteRendererProperties]}
  {emitterName=Empty, simTarget=GPUComputeSim, rendererClasses [&NiagaraSpriteRendererProperties]}
 ]
 userVariables [
  User.bIsFrontEnd {type=NiagaraBool}
  User.bIsFrontEndPreview {type=NiagaraBool}
  User.ParaGlideLeanAlpha {type=NiagaraFloat, defaultValue=1, description=""}
  User.RibbonIDOffset {type=NiagaraInt32}
  User.SkeletalMesh {
   type=NiagaraDataInterfaceSkeletalMesh
   defaultValue {
    struct=NiagaraExt_VariableValue_DataInterface
    value {dataInterface=&self:NiagaraDataInterfaceSkeletalMesh_0}
   }
  }
 ]
}
